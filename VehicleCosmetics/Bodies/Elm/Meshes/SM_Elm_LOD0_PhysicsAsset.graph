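# Physics asset for the Elm vehicle cosmetic body mesh at LOD0, giving it a simplified collision shape.
# A single body on the frame bone holds one capsule, offset from the bone origin and rotated to lie roughly horizontal.
# It has no constraints, so the asset defines no joints or articulated parts.
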PhysicsAsset /VehicleCosmetics/Bodies/Elm/Meshes/SM_Elm_LOD0_PhysicsAsset
bodies {frame [{shapeType=Capsule, length=86.9053, radius=34.5351, center=<5.2589 10.6503 0.0377>, rotation=<rot 0 -84.604 -89.965>}]}
constraints []
